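# Master Default Lit surface material for editor shapes, supporting static lighting and exposing color, tiling and roughness modulation parameters.
# ShapeColor RGB is multiplied by itself, squaring the tint.
# A Tiling scalar scales the UVs shared by the shape diffuse and normal textures.
# The diffuse texture serves as a mask: its inverted red channel is scaled by Roughness Modulation, while the normal map is sampled directly.
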Material /Engine/EditorShapes/Materials/M_ShapeMaster
graph {
 properties {bUsedWithStaticLighting=true, pixelDepthOffsetMode=PDOM_Legacy, usedShadingModels=Default Lit}
 nodes [
  n1 = VectorParameter() {parameterName=ShapeColor, defaultValue=<rgba 1 1 1 0.5>}
  Multiply(A=@n1.RGB, B=@n1.RGB)
  n2 = Multiply(A=TextureCoordinate(), B=ScalarParameter() {defaultValue=4, parameterName=Tiling})
  Multiply(
   A=OneMinus(None=TextureSample(UVs=@n2) {texture=/Engine/EditorShapes/Textures/T_ShapeDiffuse}.R)
   B=ScalarParameter() {defaultValue=0.4, parameterName=Roughness Modulation}
  ) {constB=1.5}
  TextureSample(UVs=@n2) {samplerType=SAMPLERTYPE_Normal, texture=/Engine/EditorShapes/Textures/T_ShapeNormal}
 ]
}
